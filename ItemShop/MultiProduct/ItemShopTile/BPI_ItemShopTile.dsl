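; Blueprint interface for item shop tiles, defining the calls that a tile in the multi-product shop can implement.
; It covers tile lifecycle (initialize, reset), binding an offer view model, and assigning section ID and dimensions.
; State setters handle hover with an immediate option, peek and row state.
; Other calls request presentation loading by display asset path or a 3D preview, and getters expose the offer VM, anchor and widget.
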(fn GetOfferVM
  (return 0))

(fn GetTileAnchor
  (return 0))

(fn GetTileWidget
  (return 0))

(fn InitializeTile)

(fn Request3DPreview)

(fn RequestLoadPresentations
  (params (DisplayAssetPath :type String)))

(fn ResetTile)

(fn SetHoveredState
  (params (IsHovered :type Boolean) (IsImmediate :type Boolean)))

(fn SetOfferVM
  (params (OfferVM :type Fort%20Item%20Shop%20Offer%20VM%20Object%20Reference)))

(fn SetPeekState
  (params (IsPeeked :type Boolean)))

(fn SetRowState
  (params (InRowState :type E_ItemShop_RowState%20Enum :default "NewEnumerator0")))

(fn SetSectionID
  (params (InSectionID :type String)))

(fn SetTileDimensions
  (params (Dimensions :type Vector%202D%20Structure)))
