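# Base item shop tile widget, built on UIKitModularButton with no widgets of its own, presumably for child tiles to extend.
# Tile interface events, such as setup, offer view model, state changes, preview and reset, fan out to components stored through RegisterComponents.
# Hover drives component state and button block transitions, with a brief disable window, while gaining focus broadcasts OnTileFocused.
# Helper functions choose offer description text and visibility by offer type.
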WidgetBlueprint /ItemShop/MultiProduct/ItemShopTile/WBP_ItemShopTile_Base
graphs [
 EventGraph {nodes [
   On Disable Hover() {is=title K2Node_CustomEvent}
   > Set IsHoverDisabled(IsHoverDisabled:bool=true) -> (Output_Get:bool)
   > Delay(Duration=0.2)
   > Set IsHoverDisabled(IsHoverDisabled=false)
   Event OnAddedToFocusPath!() -> (OutputDelegate, then, InFocusEvent:Focus Event Structure) {is=title K2Node_Event}
   > Call On Tile Focused!(execute:Exec, self:WBP Item Shop Tile Base Object Reference, FocusedTile:User Widget Object Reference=Self-Reference()) -> (then:Exec) {is=title K2Node_CallDelegate}
   n1 = Event SetTileDimensions!() -> (OutputDelegate, then, Dimensions:Vector 2D Structure) {is=title K2Node_Event}
   > n2 = For Each Loop(Array:Array of BPI Item Shop Tile Interfaces=Get Components() -> (Components:Array of BPI Item Shop Tile Interfaces)) -> (Array Element:BPI Item Shop Tile Interface)
   > SetTileDimensions!(execute, self:BPI Item Shop Tile Interface=@n2, Dimensions:Vector 2D Structure=@n1.Dimensions) -> (then)
   Event Request3DPreview() {is=title K2Node_Event}
   > n3 = For Each Loop(Array=Get Components())
   > Request3DPreview(self:BPI Item Shop Tile Interface=@n3)
   Event InitializeTile() {is=title K2Node_Event}
   > n4 = For Each Loop(Array=Get Components())
   > InitializeTile(self:BPI Item Shop Tile Interface=@n4)
   n5 = Event SetRowState!() -> (OutputDelegate, then, InRowState:E_ItemShop_RowState Enum=NewEnumerator4) {is=title K2Node_Event}
   > n6 = For Each Loop(Array=Get Components())
   > SetRowState!(execute, self:BPI Item Shop Tile Interface=@n6, InRowState:E_ItemShop_RowState Enum=@n5.InRowState) -> (then)
   n7 = Event SetHoveredState!() -> (OutputDelegate, then, IsHovered:bool, IsImmediate:bool) {is=title K2Node_Event}
   > n8 = For Each Loop(Array=Get Components())
   > SetHoveredState!(execute, self:BPI Item Shop Tile Interface=@n8, IsHovered:bool=@n7.IsHovered, IsImmediate:bool=@n7.IsImmediate) -> (then)
   n9 = Event SetPeekState!() -> (OutputDelegate, then, IsPeeked:bool) {is=title K2Node_Event}
   > n10 = For Each Loop(Array=Get Components())
   > SetPeekState!(execute, self:BPI Item Shop Tile Interface=@n10, IsPeeked:bool=@n9.IsPeeked) -> (then)
   n11 = Event SetSectionID!() -> (OutputDelegate, then, InSectionID:String) {is=title K2Node_Event}
   > n12 = For Each Loop(Array=Get Components())
   > SetSectionID!(execute, self:BPI Item Shop Tile Interface=@n12, InSectionID:String=@n11.InSectionID) -> (then)
   Event ResetTile() {is=title K2Node_Event}
   > n13 = For Each Loop(Array=Get Components())
   > ResetTile(self:BPI Item Shop Tile Interface=@n13)
   Event On Hovered()
   > n14 = Branch(Condition=Get IsHoverDisabled() -> (IsHoverDisabled:bool))
   SetHoveredState(execute=@n14.else, self:Self Object Reference, IsHovered=true, IsImmediate=false)
   > RequestTransitionOnBlocks!(execute, self:UIKit Modular Button Object Reference, State:EUIKitBlockVisualState Enum=Focused) -> (then)
   Event On Unhovered()
   > SetHoveredState(IsHovered=false, IsImmediate=false)
   > RequestTransitionOnBlocks(State=Unfocused)
   n15 = Event SetOfferVM!() -> (OutputDelegate, then, OfferVM:Fort Item Shop Offer VM Object Reference) {is=title K2Node_Event}
   > n16 = For Each Loop(Array=Get Components())
   > SetOfferVM!(execute, self:BPI Item Shop Tile Interface=@n16, OfferVM:Fort Item Shop Offer VM Object Reference=@n15.OfferVM) -> (then)
   n17 = Event RequestLoadPresentations!() -> (OutputDelegate, then, DisplayAssetPath:String) {is=title K2Node_Event}
   > n18 = For Each Loop(Array=Get Components())
   > RequestLoadPresentations!(execute, self:BPI Item Shop Tile Interface=@n18, DisplayAssetPath:String=@n17.DisplayAssetPath) -> (then)
  ]}
 GetOfferDescriptionText {nodes [
   n1 = GetOfferDescriptionText!() -> (then, Description:Text, OfferType:EFortItemShopOfferType Enum=LevelBundleOffer) {is=title K2Node_FunctionEntry}
   n2 = Get Description() -> (Description:Text)
   > Return Node!(
    execute
    ShortDescription:Text=Select!(
     None:Text
     LevelBundle:Text=@n2
     BattlePass:Text=@n2
     ProgressiveCosmetic:Text=Get ProgressiveCosmeticShortDescriptionText() -> (ProgressiveCosmeticShortDescriptionText:Text)
     Crew:Text=Get CrewShortDescriptionText() -> (CrewShortDescriptionText:Text)
     Special:Text=@n2
     Regular:Text=@n2
     MainItemType:Text
     MainItemTypeAndBonus:Text
     Pack:Text
     OutfitPack:Text
     PackAndBonus:Text
     Vehicle:Text
     VehicleAndBonus:Text
     Bundle:Text=@n2
     LEGOKit:Text
     Index:EFortItemShopOfferType Enum=@n1.OfferType
    ) -> (ReturnValue:Text)
   )
  ]}
 GetOfferDescriptionVisibility {nodes [
   n1 = GetOfferDescriptionVisibility!() -> (then, Description:Text, OfferType:EFortItemShopOfferType Enum=LevelBundleOffer) {is=title K2Node_FunctionEntry}
   n2 = TextIsEmpty(InText=GetOfferDescriptionText!(self, Description:Text=@n1, OfferType:EFortItemShopOfferType Enum=@n1.OfferType) -> (ShortDescription:Text) {is=title K2Node_CallFunction})
   n3 = Select(
    None:bool=true
    LevelBundle:bool=true
    BattlePass:bool=false
    ProgressiveCosmetic:bool=false
    Crew:bool=false
    Special:bool=true
    Regular:bool=false
    MainItemType:bool=false
    MainItemTypeAndBonus:bool=false
    Pack:bool=false
    OutfitPack:bool=false
    PackAndBonus:bool=false
    Vehicle:bool=false
    VehicleAndBonus:bool=false
    Bundle:bool=false
    LEGOKit:bool=false
    Index=@n1.OfferType
   ) -> (ReturnValue:bool)
   > Return Node!(
    execute
    Visibility:ESlateVisibility Enum=Select!(Option 0:ESlateVisibility Enum=SelfHitTestInvisible, Option 1:ESlateVisibility Enum=Collapsed, Index:bool=OR Boolean(A=@n2, B=@n3)) -> (ReturnValue:ESlateVisibility Enum)
   )
  ]}
 GetOfferVM {nodes [
   GetOfferVM() {is=title K2Node_FunctionEntry}
   > Return Node!(execute, OfferVM:Fort Item Shop Offer VM Object Reference)
  ]}
 GetTileAnchor {nodes [
   GetTileAnchor() {is=title K2Node_FunctionEntry}
   > Return Node!(execute, Anchor:Vector 2D Structure)
  ]}
 GetTileWidget {nodes [
   GetTileWidget() {is=title K2Node_FunctionEntry}
   > Return Node!(execute, NewParam:User Widget Object Reference=Self-Reference())
  ]}
 OnOfferHovered {nodes [OnOfferHovered!() -> (then, OfferId:String) {is=title K2Node_FunctionEntry}]}
 OnRequestScrollToRow {nodes [OnRequestScrollToRow() {is=title K2Node_FunctionEntry}]}
 OnTileFocused {nodes [OnTileFocused!() -> (then, FocusedTile:User Widget Object Reference) {is=title K2Node_FunctionEntry}]}
 RegisterComponents {nodes [
   n1 = RegisterComponents!() -> (then, Components:Array of BPI Item Shop Tile Interfaces) {is=title K2Node_FunctionEntry}
   > Set Components(Components:Array of BPI Item Shop Tile Interfaces=@n1) -> (Output_Get:Array of BPI Item Shop Tile Interfaces)
  ]}
]
variables [
 1x1TileSizeRatio {type=Vector2D, default=<1 1.5938>}
 Components {type="array<object>", default []}
 CrewShortDescriptionText {type=string, default=Subscription}
 IsHoverDisabled {type=boolean}
 OnOfferHovered {}
 OnRequestScrollToRow {}
 OnTileFocused {}
 ProgressiveCosmeticShortDescriptionText {type=string, default=Unlockable styles each month}
]
widgets {info {inheritedWidgetCount=0, parentClass=/Script/FortniteUI.UIKitModularButton, rootWidgetClass=null, widgetCount=0}, widgets []}
